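; DEMO_WiryPerk is a demo Blueprint in the WiryPerk emote VFX folder that shows an emote animation on a character mesh component.
; On BeginPlay it plays the male CMM SatireCane_Ode emote, looping, on the M_MED_Base component.
; The animation comes from the SatireCane_Ode emote folder rather than WiryPerk.
; Overlap, Tick and the construction script are empty.
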(event EventBeginPlay
  (Components|Animation|PlayAnimation (Variables|Default|GetM_MED_Base) "/BRCosmetics/Animation/Game/MainPlayer/Emotes/SatireCane_Ode/CMM/Emote_SatireCane_Ode_CMM_M.Emote_SatireCane_Ode_CMM_M" true))

(event Collision|EventActorBeginOverlap
  (params (OtherActor :type Actor%20Object%20Reference)))

(event EventTick
  (params (DeltaSeconds :type Float%20%28single-precision%29)))

(fn ConstructionScript)
