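# Niagara impact effect for the WiryPerk emote: glowing block bursts and a "GG" graphic that collapses into falling pieces.
# Two grid-spawned block emitters fire first, then a delayed GG_Text sprite animates through dynamic material parameters, with lights.
# Later-spawned GPU and CPU collapse grids stay hidden until mid-life, then fall; the GPU grid samples T_WiryPerk_GG for color.
# Flash, shockwave, smoke, flare and star emitters have no renderers.
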NiagaraSystem /BRCosmetics/Animation/Game/MainPlayer/Emotes/WiryPerk/VFX/NS_Emote_WiryPerk_Impact
dependencies {
 usedDataInterfaces [
  &NiagaraDataInterfaceCurve
  &NiagaraDataInterfaceVectorCurve
  &NiagaraDataInterfaceColorCurve
  /Script/Niagara.NiagaraDataInterfaceTexture
  &NiagaraDataInterfaceArrayColor
  &NiagaraDataInterfaceVector2DCurve
 ]
 usedDynamicInputs [&Multiply_Float, &ColorFromCurve]
 usedModules [
  &self:SystemSpawnScript.SystemScriptSource.SystemScriptGraph.NiagaraNodeAssignment_0.SetVariables_#1
  &SystemState
 ]
}
emitters {
 3DBlocks_Horiz {
  emitterUpdateScript [
   &EmitterState(Life Cycle Mode=::NewEnumerator1, Loop Behavior=::NewEnumerator1, Loop Delay=0.05, Loop Duration=0.1, UseLoopDelay=false)
   /Niagara/Modules/Spawn/Location/SpawnParticlesInGrid(X Count=8, Y Count=1, Z Count=8)
   &SpawnBurst_Instantaneous(
    Spawn Count=&MakeCustomIntFromBool(Bool=$User.bIsFrontEndPreview, True INT=0, False INT=&RandomRangeInt(Minimum=15, Maximum=20))
    Use Spawn Probability=false
    UseLoopCountLimit=false
   ) {enabled=false}
  ]
  particleSpawnScript [
   &InitializeParticle(
    Color Channel Mode:ENiagaraChannelCorrelation=::NewEnumerator0
    Color Maximum=<rgba 1.5 1.5 1.5 1>
    Color Minimum=<rgba 1 1 1 1>
    Color Mode=::NewEnumerator2
    Lifetime Max=1
    Lifetime Min=0.2
    Lifetime Mode=::NewEnumerator1
    Mesh Scale Mode=::NewEnumerator3
    Mesh Uniform Scale=0.4
    Sprite Size Mode=::NewEnumerator4
    UsePositionOffset=false
    Write Material Random=true
   )
   /Niagara/Modules/Spawn/Location/V2/GridLocation(
    Module.Jitter Inside Grid Cells=false
    Module.Override Random Seed=false
    Offset:Vector3f=&RandomRangeVector(Minimum=<-1.5 -1 0>, Maximum=<1.5 1 1>)
    XYZ Dimensions:Vector3f=<5 5 5>
   )
   &AddVelocity(Velocity=&RandomRangeVector(Minimum=<-250 -250 300>, Maximum=<250 250 600>)) {module=AddVelocity001}
  ]
  particleUpdateScript [
   /SetVariables(Particles.AlphaDistance:NiagaraFloat=&NormalizeDistanceRange(
     Start Position:NiagaraPosition=/Niagara/DynamicInputs/Helpers/SimulationPosition()
     End Position:NiagaraPosition=$Particles.Position
     Distance:NiagaraFloat=50
     Invert Normalized Range:NiagaraBool=false
    )) {enabled=false, isSetParameters=true, module=SetVariables_#2}
   /SetVariables(
    Particles.GlowColor=&MakeLinearColorFromVectorAndFloat(
     "Vector (RGB)"=&MakeVectorFromLinearColorRGB(Color=<rgba 0 1 0.0833 1>)
     "Float (Alpha)"=&FloatFromCurve(FloatCurve=DataInterface() {propertyValues {Curve {keys [<at 0.5 1>,<at 1 0>]}}}, Scale Curve=0.08)
    )
    Particles.GlowSize=&Multiply_Vector2D(A=$Particles.SpriteSize, B=<8 8>)
   ) {isSetParameters=true, module=SetVariables_#3}
   &ScaleColor(
    Scale Alpha=$Particles.AlphaDistance
    Scale RGB=&MakeVectorFromLinearColorRGB(Color=&LerpLinearColors(StartColor=<rgba 0.3419 0.8632 0.0437 1>, EndColor=<rgba 0.0316 0 0.1 1>, LerpFactor=$Particles.AlphaDistance))
    ScaleA=true
    ScaleRGB=true
   ) {enabled=false, module=ScaleColor001}
   &SpriteRotationRate(
    Delta Time:NiagaraFloat=$Particles.NormalizedAge
    Rotation Rate=&RandomRangeFloat(Minimum=-25, Maximum=25)
   ) {enabled=false}
   &ScaleVelocity(Velocity Scale=&VectorFromCurve(VectorCurve=DataInterface() {propertyValues {
       XCurve {keys [{arriveTangent=610.798, interpMode=RCIM_Cubic, leaveTangent=610.798, tangentMode=RCTM_User, time=0.1, value=0}, <at 0.15 20>]}
       YCurve {keys [{arriveTangent=610.798, interpMode=RCIM_Cubic, leaveTangent=610.798, tangentMode=RCTM_User, time=0.1, value=0}, <at 0.15 20>]}
       ZCurve {keys [
         {arriveTangent=610.798, interpMode=RCIM_Cubic, leaveTangent=610.798, tangentMode=RCTM_User, time=0.1, value=0}
         <at 0.15 20>
         {arriveTangent=-141.1765, interpMode=RCIM_Cubic, leaveTangent=-141.1765, time=0.65, value=20}
         {arriveTangent=-536.5441, tangentMode=RCTM_User, time=1, value=-100}
        ]}
      }})) {enabled=false}
   &GravityForce()
   &Drag(Use Linear Drag=true, Use Rotational Drag=false)
   &ScaleVelocity(
    Coordinate Space:ENiagaraCoordinateSpace=::World
    Velocity Scale=&VectorFromCurve(
     VectorCurve=DataInterface() {propertyValues {XCurve {keys [<at 0 3 RCIM_Cubic>,<at 1 0>]}, YCurve {keys [<at 0 3 RCIM_Cubic>,<at 1 0>]}, ZCurve {keys [<at 0 3 RCIM_Cubic>,<at 1 0>]}}}
     Scale Curve=&RandomRangeVector(Minimum=<1 1 1>, Maximum=<1 1 1.8>)
    )
   ) {enabled=false, module=ScaleVelocity001}
   &SolveForcesAndVelocity(Clamp Velocity=true, Limit Acceleration=false)
   &ParticleState()
   &ScaleMeshSize(Scale Factor=&VectorFromCurve(VectorCurve=DataInterface() {propertyValues {
       XCurve {keys [<at 0.75 1 RCIM_Cubic>,<at 1 0>]}
       YCurve {keys [<at 0.75 1 RCIM_Cubic>,<at 1 0>]}
       ZCurve {keys [<at 0.75 1 RCIM_Cubic>,<at 1 0>]}
      }}))
   &ScaleMeshSizeBySpeed(
    Max Scale Factor=<1 1 1>
    Min Scale Factor:Vector3f=<0.1 0.1 0.1>
    Sample Scale Factor By Curve=true
    Scale Factor Curve=DataInterface() {propertyValues {Curve {defaultValue=0, keys [<at 0 1 RCIM_Cubic>, {tangentMode=RCTM_User, time=1, value=0.5}]}}}
    Speed Threshold=250
   ) {enabled=false}
   &UpdateMeshOrientation(
    Rotation Rate=&FloatFromCurve(
     FloatCurve=DataInterface() {propertyValues {Curve {defaultValue=1, keys [{tangentMode=RCTM_User, time=0.2, value=0}, {tangentMode=RCTM_User, time=0.5, value=1}]}}}
     CurveIndex=$Emitter.Age
    )
    Rotation Vector=&RandomRangeVector(Minimum=<-1 -1 -1>, Maximum=<1 1 1>)
   ) {enabled=false}
   &ScaleColor(
    Linear Color Curve=DataInterface() {propertyValues {
      AlphaCurve {keys [<at 0.8001 1>,<at 1 1>]}
      BlueCurve {keys [<at 0.1582 0>,<at 0.338 0>,<at 0.5201 1.3716>,<at 0.8661 1>]}
      GreenCurve {keys [<at 0.1582 10>,<at 0.338 0.656>,<at 0.5201 0.1464>,<at 0.8661 0>]}
      RedCurve {keys [<at 0.1582 2.282>,<at 0.338 2>,<at 0.5201 2>,<at 0.8661 0.2993>]}
     }}
    Scale Mode=::NewEnumerator2
   ) {module=ScaleColor002}
   &ScaleColor(
    Scale RGB=&VectorFromCurve(VectorCurve=DataInterface() {propertyValues {XCurve {keys [<at 0 5 RCIM_Cubic>,<at 1 3>]}, YCurve {keys [<at 0 5 RCIM_Cubic>,<at 1 3>]}, ZCurve {keys [<at 0 5 RCIM_Cubic>,<at 1 3>]}}})
    ScaleA=true
    ScaleRGB=true
   )
   &FN_Light_Attributes(
    Color=&MultiplyLinearColor(
     Linear Color=<rgba 3 0 2.6248 1>
     Scale Factor=&MakeVector4_FromLinearColor(LinearColor=&LerpLinearColors(StartColor=<rgba 1 1 1 1>, EndColor=$System.TOD Color, LerpFactor=0.5))
    )
    Create Color Attribute=true
    Create Exponent Attribute=true
    Create Position Attribute=true
    Create Radius Attribute=true
    Exponent=2
    Position=<0 -25 200>
    Radius=&FN_FrontEndFloatModifier(Gameplay=80, Front End=80)
    Scale Color=&FloatFromCurve(
     FloatCurve=DataInterface() {propertyValues {Curve {keys [
         {arriveTangent=-0.5504, interpMode=RCIM_Cubic, leaveTangent=-0.5504, tangentMode=RCTM_User, time=0.5, value=1}
         {arriveTangent=-0.5426, interpMode=RCIM_Cubic, leaveTangent=-0.5426, tangentMode=RCTM_User, time=1, value=0}
        ]}}}
     Scale Curve=1.5
    )
   )
  ]
  rendererClasses [
   &NiagaraSpriteRendererProperties
   &NiagaraMeshRendererProperties
   &NiagaraLightRendererProperties
  ]
  renderers [
   {rendererClass=&NiagaraSpriteRendererProperties}
   {rendererClass=&NiagaraSpriteRendererProperties, rendererIndex=1}
   {rendererClass=&NiagaraMeshRendererProperties, rendererIndex=2}
   {rendererClass=&NiagaraLightRendererProperties, rendererIndex=3}
  ]
 }
 3DBlocks_Vert {
  emitterUpdateScript [
   &EmitterState(Life Cycle Mode=::NewEnumerator1, Loop Behavior=::NewEnumerator1, Loop Duration=0.1, UseLoopDelay=true)
   &SpawnParticlesInGrid(X Count=6, Y Count=1)
   &SpawnBurst_Instantaneous(
    Spawn Count=&MakeCustomIntFromBool(Bool=$User.bIsFrontEndPreview, True INT=0, False INT=&RandomRangeInt(Minimum=15, Maximum=20))
    Use Spawn Probability=false
    UseLoopCountLimit=false
   ) {enabled=false}
  ]
  particleSpawnScript [
   &InitializeParticle(
    Color Maximum=<rgba 1.5 1.5 1.5 1>
    Color Minimum=<rgba 1 1 1 1>
    Color Mode=::NewEnumerator2
    Lifetime Max=0.5
    Lifetime Min=0.25
    Lifetime Mode=::NewEnumerator1
    Mesh Scale Mode=::NewEnumerator3
    Mesh Uniform Scale=0.4
    Sprite Size Mode=::NewEnumerator4
    UsePositionOffset=false
    Write Material Random=true
   )
   &GridLocation(Module.Jitter Inside Grid Cells=false, Module.Override Random Seed=false, Offset:Vector3f=<0 -80 0>, XYZ Dimensions:Vector3f=<10 1 10>)
   &AddVelocity(Velocity=&RandomRangeVector(Minimum=<0 0 250>, Maximum=<1 1 600>)) {module=AddVelocity001}
   &Color(Color=&MakeLinearColorFromVectorAndFloat("Vector (RGB)"=Linked() {linkedVariable {name=Particles.Position, type=NiagaraPosition}}, "Float (Alpha)"=1)) {enabled=false}
   &ShapeLocation(MaskSpawnGroup=false, Override Seed=false, Sphere Radius=5, Sphere Surface Distribution=1, UseOwnerScale=true)
  ]
  particleUpdateScript [
   /SetVariables(Particles.AlphaDistance:NiagaraFloat=&NormalizeDistanceRange(
     Start Position:NiagaraPosition=&SimulationPosition()
     End Position:NiagaraPosition=$Particles.Position
     Distance:NiagaraFloat=50
     Invert Normalized Range:NiagaraBool=false
    )) {enabled=false, isSetParameters=true, module=SetVariables_#2}
   /SetVariables(
    Particles.GlowColor=&MakeLinearColorFromVectorAndFloat(
     "Vector (RGB)"=&MakeVectorFromLinearColorRGB(Color=<rgba 0 1 0.0833 1>)
     "Float (Alpha)"=&FloatFromCurve(FloatCurve=DataInterface() {propertyValues {Curve {keys [<at 0.5 1>,<at 1 0>]}}}, Scale Curve=0.08)
    )
    Particles.GlowSize=&Multiply_Vector2D(A=$Particles.SpriteSize, B=<8 8>)
   ) {isSetParameters=true, module=SetVariables_#3}
   &ScaleColor(
    Scale Alpha=$Particles.AlphaDistance
    Scale RGB=&MakeVectorFromLinearColorRGB(Color=&LerpLinearColors(StartColor=<rgba 0 1 0 1>, EndColor=<rgba 0.5 0 0.5 1>, LerpFactor=$Particles.AlphaDistance))
    ScaleA=true
    ScaleRGB=true
   ) {enabled=false, module=ScaleColor001}
   &ScaleVelocity(Velocity Scale=&VectorFromCurve(VectorCurve=DataInterface() {propertyValues {
       XCurve {keys [{arriveTangent=610.798, interpMode=RCIM_Cubic, leaveTangent=610.798, tangentMode=RCTM_User, time=0.1, value=0}, <at 0.15 20>]}
       YCurve {keys [{arriveTangent=610.798, interpMode=RCIM_Cubic, leaveTangent=610.798, tangentMode=RCTM_User, time=0.1, value=0}, <at 0.15 20>]}
       ZCurve {keys [
         {arriveTangent=610.798, interpMode=RCIM_Cubic, leaveTangent=610.798, tangentMode=RCTM_User, time=0.1, value=0}
         <at 0.15 20>
         {arriveTangent=-141.1765, interpMode=RCIM_Cubic, leaveTangent=-141.1765, time=0.65, value=20}
         {arriveTangent=-536.5441, tangentMode=RCTM_User, time=1, value=-100}
        ]}
      }})) {enabled=false}
   &Drag(
    Drag=&FloatFromCurve(FloatCurve=DataInterface() {propertyValues {Curve {defaultValue=1, keys [{tangentMode=RCTM_User, time=0, value=0}, {tangentMode=RCTM_User, time=0.25, value=1}]}}})
    Use Linear Drag=true
    Use Rotational Drag=true
   ) {enabled=false}
   &SolveForcesAndVelocity(
    Clamp Velocity=false
    Limit Acceleration=false
    Speed Limit:NiagaraFloat=&FloatFromCurve(FloatCurve=DataInterface() {propertyValues {Curve {keys [<at 0.2 500 RCIM_Cubic>,<at 0.5 0>]}}})
   )
   &ParticleState()
   &ScaleMeshSize(Scale Factor=&VectorFromCurve(VectorCurve=DataInterface() {propertyValues {
       XCurve {keys [<at 0.75 1 RCIM_Cubic>,<at 1 1>]}
       YCurve {keys [<at 0.75 1 RCIM_Cubic>,<at 1 1>]}
       ZCurve {keys [<at 0.75 1 RCIM_Cubic>,<at 1 1>]}
      }}))
   &ScaleMeshSizeBySpeed(Max Scale Factor=<1 1 1>, Min Scale Factor:Vector3f=<0.1 0.1 0.1>, Speed Threshold=300) {enabled=false}
   &UpdateMeshOrientation(
    Rotation Rate=&FloatFromCurve(
     FloatCurve=DataInterface() {propertyValues {Curve {defaultValue=1, keys [{tangentMode=RCTM_User, time=0.2, value=0}, {tangentMode=RCTM_User, time=0.5, value=1}]}}}
     CurveIndex=$Emitter.Age
    )
    Rotation Vector=&RandomRangeVector(Minimum=<-1 -1 -1>, Maximum=<1 1 1>)
   ) {enabled=false}
   &ScaleColor(
    Linear Color Curve=DataInterface() {propertyValues {
      AlphaCurve {keys [<at 0.8092 1>,<at 1 1>]}
      BlueCurve {keys [<at 0.2424 0.6518>,<at 0.486 0>,<at 0.8114 0.1647>]}
      GreenCurve {keys [<at 0.2424 0>,<at 0.486 0.7211>,<at 0.8114 1>]}
      RedCurve {keys [<at 0.2424 2>,<at 0.486 2>,<at 0.8114 0>]}
     }}
    Scale Mode=::NewEnumerator2
   ) {module=ScaleColor002}
   &ScaleColor(
    Scale RGB=&VectorFromCurve(VectorCurve=DataInterface() {propertyValues {XCurve {keys [<at 0 5 RCIM_Cubic>,<at 1 3>]}, YCurve {keys [<at 0 5 RCIM_Cubic>,<at 1 3>]}, ZCurve {keys [<at 0 5 RCIM_Cubic>,<at 1 3>]}}})
    ScaleA=true
    ScaleRGB=true
   )
  ]
  rendererClasses [&NiagaraSpriteRendererProperties, &NiagaraMeshRendererProperties]
  renderers [
   {rendererClass=&NiagaraSpriteRendererProperties}
   {rendererClass=&NiagaraSpriteRendererProperties, rendererIndex=1}
   {rendererClass=&NiagaraMeshRendererProperties, rendererIndex=2}
  ]
 }
 DirectionalFlare001 {rendererClasses [], renderers []}
 Flash {rendererClasses [], renderers []}
 GG_Collapse_CPU {
  emitterUpdateScript [
   &EmitterState(
    Inactive Response=::NewEnumerator2
    Life Cycle Mode=::NewEnumerator1
    Loop Behavior=::NewEnumerator1
    Loop Delay=0.4
    Loop Duration=0.1
    UseLoopDelay=true
   )
   &SpawnParticlesInGrid(X Count=5, Y Count=1, Z Count=5)
  ]
  particleSpawnScript [
   &InitializeParticle(
    Color=&RandomRangeLinearColor(Minimum=<rgba 1 1 1 1>, Maximum=<rgba 2 2 2 1>)
    Lifetime=&RandomRangeFloat(Minimum=1.5, Maximum=2)
    Mesh Scale Mode=::NewEnumerator3
    Mesh Uniform Scale=0.4
    UsePositionOffset=false
    Write Material Random=true
   )
   &Color(Color=&SelectLinearColorFromArray(
     Array Sampling Mode=::NewEnumerator2
     Direct Array Index=?
     Interpolation=&ReturnNormalizedExecIndex()
     Color Selection Array=DataInterface() {propertyValues {ColorData [<rgba 0 1 0.2276 1>, <rgba 0.8229 0 0.72 1>, <rgba 0.8229 0 0.72 1>, <rgba 0.8229 0 0.72 1>, <rgba 0.8229 0 0.72 1>, <rgba 0 0.1284 1 1>]}}
     Random Seed=?
    ))
   &GridLocation(
    Module.Jitter Inside Grid Cells=true
    Module.Override Random Seed=false
    Offset:Vector3f=&RandomRangeVector(Minimum=<-35 0 20>, Maximum=<35 1 -20>)
    Random Placement Seed=100
    XYZ Dimensions:Vector3f=<1.5 1 1.75>
   )
   &ApplyInitialForces()
  ]
  particleUpdateScript [
   /SetVariables(
    Particles.GlowColor=&MakeLinearColorFromVectorAndFloat(
     "Vector (RGB)"=&MakeVectorFromLinearColorRGB(Color=<rgba 1 0 1 1>)
     "Float (Alpha)"=&FloatFromCurve(FloatCurve=DataInterface() {propertyValues {Curve {keys [<at 0.2 0.5 RCIM_Cubic>,<at 1 0>]}}}, Scale Curve=0.08)
    )
    Particles.GlowSize=&Multiply_Vector2D(A=$Particles.SpriteSize, B=<2 2>)
   ) {isSetParameters=true, module=SetVariables_#3}
   &GravityForce(Gravity=&RandomRangeVector(Minimum=<0 0 -500>, Maximum=<1 1 -980>)) {module=GravityForce001}
   &SolveForcesAndVelocity(
    Clamp Velocity=true
    Limit Acceleration=false
    Speed Limit:NiagaraFloat=&FloatFromCurve(FloatCurve=DataInterface() {propertyValues {Curve {defaultValue=1, keys [<at 0.5 0 RCIM_Cubic>, {tangentMode=RCTM_User, time=1, value=500}]}}})
   )
   &ParticleState()
   &ScaleMeshSize(Scale Factor=&VectorFromCurve(VectorCurve=DataInterface() {propertyValues {
       XCurve {keys [<at 0.75 1 RCIM_Cubic>,<at 1 0>]}
       YCurve {keys [<at 0.75 1 RCIM_Cubic>,<at 1 0>]}
       ZCurve {keys [<at 0.75 1 RCIM_Cubic>,<at 1 0>]}
      }}))
   &ScaleSpriteSize(
    Non-Uniform Curve Sprite Scale=DataInterface() {propertyValues {
      XCurve {keys [{tangentMode=RCTM_User, time=0, value=0.5}, {tangentMode=RCTM_User, time=0.2, value=1}]}
      YCurve {keys [<at 0 1 RCIM_Cubic>, {tangentMode=RCTM_User, time=0.2, value=1}]}
     }}
    Scale Sprite Size Mode=::NewEnumerator3
   ) {enabled=false}
   &UpdateMeshOrientation(
    Rotation Rate=&FloatFromCurve(
     FloatCurve=DataInterface() {propertyValues {Curve {defaultValue=1, keys [{tangentMode=RCTM_User, time=0.2, value=0}, {tangentMode=RCTM_User, time=0.5, value=1}]}}}
     CurveIndex=$Emitter.Age
    )
    Rotation Vector=&RandomRangeVector(Minimum=<-1 -1 -1>, Maximum=<1 1 1>)
   ) {enabled=false}
   &ScaleColor(
    Scale Alpha=&FloatFromCurve(FloatCurve=DataInterface() {propertyValues {Curve {
        defaultValue=0
        keys [
         {arriveTangent=62.1041, interpMode=RCIM_Cubic, leaveTangent=62.1041, tangentMode=RCTM_User, time=0.5, value=0}
         {tangentMode=RCTM_User, time=0.55, value=1}
        ]
       }}})
    Scale RGB=&Multiply_VectorByFloat(
     Vector=&MakeVectorFromLinearColorRGB(Color=$Particles.SampleTexture.SampledColor)
     Float=&FloatFromCurve(FloatCurve=DataInterface() {propertyValues {Curve {keys [<at 0.2 5 RCIM_Cubic>,<at 1 4>]}}})
    )
    ScaleA=true
    ScaleRGB=true
   )
   &CameraOffset(Camera Offset Amount=-50)
   &OffsetPosition(Position Offset=<0 -50 200>)
  ]
  rendererClasses [&NiagaraSpriteRendererProperties, &NiagaraMeshRendererProperties]
  renderers [
   {rendererClass=&NiagaraSpriteRendererProperties}
   {rendererClass=&NiagaraMeshRendererProperties, rendererIndex=1}
  ]
 }
 GG_Collapse_GPU {
  simTarget=GPUComputeSim
  emitterUpdateScript [
   &EmitterState(
    Inactive Response=::NewEnumerator2
    Life Cycle Mode=::NewEnumerator1
    Loop Behavior=::NewEnumerator1
    Loop Delay=0.4
    Loop Duration=0.1
    UseLoopDelay=true
   )
   &SpawnParticlesInGrid(X Count=15, Y Count=1, Z Count=15)
   &SpawnBurst_Instantaneous(Use Spawn Probability=false, UseLoopCountLimit=false) {enabled=false}
  ]
  particleSpawnScript [
   &InitializeParticle(
    Color=&RandomRangeLinearColor(Minimum=<rgba 0.5 0.5 0.5 1>, Maximum=<rgba 1 1 1 1>)
    Lifetime=&RandomRangeFloat(Minimum=1.5, Maximum=2)
    Mesh Scale Mode=::NewEnumerator3
    Mesh Uniform Scale=0.4
    UsePositionOffset=false
    Write Material Random=true
   )
   &AddVelocity(Velocity=<0 0 1100>) {enabled=false, module=AddVelocity001}
   &GridLocation(Module.Jitter Inside Grid Cells=false, Module.Override Random Seed=false, XYZ Dimensions:Vector3f=<1 1 1>)
   /Niagara/Modules/Update/Textures/SampleTexture(
    Rotation Axis:Vector3f=<0 0.25 0>
    Rotation Mode:ENiagara_RotationMode=::NewEnumerator1
    Texture=DataInterface() {propertyValues {
      Texture=&/T_WiryPerk_GG
      TextureUserParameter {parameter {name=None, typeDefHandle {registeredTypeIndex=2}, varData []}}
     }}
    Transform UV Into Position=true
    UV:Vector2f=&MakeVector2D(
     X=&MakeFloatFromVector(Vector=$Output.GridLocation.GridUVW, Channel=::NewEnumerator0)
     Y=&MakeFloatFromVector(Vector=$Output.GridLocation.GridUVW, Channel=::NewEnumerator2)
    )
    UV Scale:Vector2f=<125 75>
   )
   &Color(
    Color=$Particles.SampleTexture.SampledColor
    Scale Color=&RandomRangeVector(Minimum=<1 1 1>, Maximum=<1.2 1.2 1.2>)
   )
   &GravityForce() {enabled=false}
   &PointForce(Force Strength=300, Module.Use Falloff=false, Module.UseLinearForce=false) {enabled=false}
   &ApplyInitialForces()
  ]
  particleUpdateScript [
   /SetVariables(
    Particles.GlowColor=&MakeLinearColorFromVectorAndFloat(
     "Vector (RGB)"=&MakeVectorFromLinearColorRGB(Color=<rgba 1 0 1 1>)
     "Float (Alpha)"=&FloatFromCurve(FloatCurve=DataInterface() {propertyValues {Curve {keys [<at 0.2 0.5 RCIM_Cubic>,<at 1 0>]}}}, Scale Curve=0.08)
    )
    Particles.GlowSize=&Multiply_Vector2D(A=$Particles.SpriteSize, B=<2 2>)
   ) {isSetParameters=true, module=SetVariables_#3}
   &GravityForce(Gravity=&RandomRangeVector(Minimum=<0 0 -500>, Maximum=<1 1 -1800>)) {module=GravityForce001}
   &SolveForcesAndVelocity(
    Clamp Velocity=true
    Limit Acceleration=false
    Speed Limit:NiagaraFloat=&FloatFromCurve(FloatCurve=DataInterface() {propertyValues {Curve {defaultValue=1, keys [<at 0.5 0 RCIM_Cubic>, {tangentMode=RCTM_User, time=1, value=500}]}}})
   )
   &ParticleState()
   &ScaleMeshSize(Scale Factor=&VectorFromCurve(VectorCurve=DataInterface() {propertyValues {
       XCurve {keys [<at 0.75 1 RCIM_Cubic>,<at 1 0>]}
       YCurve {keys [<at 0.75 1 RCIM_Cubic>,<at 1 0>]}
       ZCurve {keys [<at 0.75 1 RCIM_Cubic>,<at 1 0>]}
      }}))
   &ScaleColor(
    Scale Alpha=&FloatFromCurve(FloatCurve=DataInterface() {propertyValues {Curve {
        defaultValue=0
        keys [
         {arriveTangent=62.1041, interpMode=RCIM_Cubic, leaveTangent=62.1041, tangentMode=RCTM_User, time=0.5, value=0}
         {tangentMode=RCTM_User, time=0.55, value=1}
        ]
       }}})
    Scale RGB=&Multiply_VectorByFloat(
     Vector=&MakeVectorFromLinearColorRGB(Color=$Particles.SampleTexture.SampledColor)
     Float=&FloatFromCurve(FloatCurve=DataInterface() {propertyValues {Curve {keys [<at 0.25 0.1 RCIM_Cubic>,<at 1 20>]}}})
    )
    ScaleA=true
    ScaleRGB=true
   )
   &CameraOffset(Camera Offset Amount=-50)
   &OffsetPosition(Position Offset=<0 -80 200>)
  ]
  rendererClasses [&NiagaraSpriteRendererProperties, &NiagaraMeshRendererProperties]
  renderers [
   {rendererClass=&NiagaraSpriteRendererProperties}
   {rendererClass=&NiagaraMeshRendererProperties, rendererIndex=1}
  ]
 }
 GG_Text {
  emitterUpdateScript [
   &EmitterState(
    Inactive Response=::NewEnumerator2
    Life Cycle Mode=::NewEnumerator1
    Loop Behavior=::NewEnumerator1
    Loop Delay=0.2
    Loop Duration=0.1
    UseLoopDelay=true
   )
   &SpawnBurst_Instantaneous(Use Spawn Probability=false, UseLoopCountLimit=false)
  ]
  particleSpawnScript [
   &InitializeParticle(
    Lifetime=1.4
    Mesh Scale Mode=::NewEnumerator3
    Mesh Uniform Scale=0.4
    Sprite Rotation Angle=0.75
    Sprite Rotation Mode=::NewEnumerator1
    Sprite Size=<75 125>
    Sprite Size Mode=::NewEnumerator4
    UsePositionOffset=false
    Write Material Random=true
   )
   &SpriteFacingAndAlignment(Sprite Alignment=<0 0 1>, Sprite Facing=<0 1 0>, WriteAlignment=true, WriteFacing=true)
  ]
  particleUpdateScript [
   /SetVariables(
    Particles.GlowColor=&MakeLinearColorFromVectorAndFloat(
     "Vector (RGB)"=&MakeVectorFromLinearColorRGB(Color=<rgba 1 0 1 1>)
     "Float (Alpha)"=&FloatFromCurve(FloatCurve=DataInterface() {propertyValues {Curve {keys [<at 0.2 0.5 RCIM_Cubic>,<at 1 0>]}}}, Scale Curve=0.08)
    )
    Particles.GlowSize=&Multiply_Vector2D(A=$Particles.SpriteSize, B=<1.5 1.5>)
   ) {isSetParameters=true, module=SetVariables_#3}
   &SolveForcesAndVelocity(
    Clamp Velocity=true
    Limit Acceleration=false
    Speed Limit:NiagaraFloat=&FloatFromCurve(FloatCurve=DataInterface() {propertyValues {Curve {keys [{arriveTangent=11.1111, interpMode=RCIM_Cubic, time=0.15, value=5000}, {arriveTangent=-89865.0859, tangentMode=RCTM_User, time=0.2, value=0}]}}})
   )
   &ParticleState()
   &ScaleColor(
    Scale Alpha=&FloatFromCurve(FloatCurve=DataInterface() {propertyValues {Curve {keys [<at 0.8 1 RCIM_Cubic>,<at 1 0>]}}})
    Scale RGB=&VectorFromCurve(VectorCurve=DataInterface() {propertyValues {
       XCurve {keys [<at 0 10 RCIM_Cubic>,<at 0.2 1>]}
       YCurve {keys [<at 0 1 RCIM_Cubic>,<at 0.2 1>]}
       ZCurve {keys [<at 0 10 RCIM_Cubic>,<at 0.2 1>]}
      }})
    ScaleA=true
    ScaleRGB=true
   )
   &OffsetPosition(Position Offset=<0 -80 200>)
   &DynamicMaterialParameters(
    Index 0 Param 1=&FloatFromCurve(FloatCurve=DataInterface() {propertyValues {Curve {defaultValue=1, keys [<at 0 -1>,<at 0.15 1.4795>,<at 0.7 1.5>,<at 1 -0.9>]}}})
    Index 0 Param 2=&FloatFromCurve(FloatCurve=DataInterface() {propertyValues {Curve {keys [<at 0.75 0 RCIM_Cubic>,<at 0.8 1>]}}})
    Index 0 Param 3=&FloatFromCurve(FloatCurve=DataInterface() {propertyValues {Curve {keys [<at 0.75 0>,<at 0.8 1 RCIM_Cubic>]}}})
    Index 0 Param 4=&FloatFromCurve(FloatCurve=DataInterface() {propertyValues {Curve {keys [<at 0.5 20>,<at 1 8>]}}})
    Index 1 Param 2=&FloatFromCurve(FloatCurve=DataInterface() {propertyValues {Curve {keys [<at 0.7 0.15>,<at 1 1>]}}})
    Index 1 Param 3=&FloatFromCurve(FloatCurve=DataInterface() {propertyValues {Curve {defaultValue=1, keys [<at 0.15 1>, <at 0.2 0.1>, {tangentMode=RCTM_User, time=0.79, value=0.1}, {tangentMode=RCTM_User, time=0.8, value=1}]}}})
    Module.Index0Float0Write=true
    Module.Index0Float1Write=true
    Module.Index0Float2Write=true
    Module.Index0Float3Write=true
    Module.Index1Float0Write=true
    Module.Index1Float1Write=true
    Module.Index1Float2Write=true
    Module.Index1Float3Write=true
    Write Parameter Index 1=::NewEnumerator1
   )
   &FN_Light_Attributes(
    Color=&MultiplyLinearColor(
     Linear Color=<rgba 3 0 2.6248 1>
     Scale Factor=&MakeVector4_FromLinearColor(LinearColor=&LerpLinearColors(StartColor=<rgba 1 1 1 1>, EndColor=$System.TOD Color, LerpFactor=0.5))
    )
    Create Color Attribute=true
    Create Exponent Attribute=true
    Create Position Attribute=true
    Create Radius Attribute=true
    Exponent=2
    Position=<0 -25 200>
    Radius=110
    Scale Color=&FloatFromCurve(
     FloatCurve=DataInterface() {propertyValues {Curve {keys [
         {arriveTangent=-0.5504, interpMode=RCIM_Cubic, leaveTangent=-0.5504, tangentMode=RCTM_User, time=0.5, value=1}
         {arriveTangent=-0.5426, interpMode=RCIM_Cubic, leaveTangent=-0.5426, tangentMode=RCTM_User, time=1, value=0}
        ]}}}
     Scale Curve=1.5
    )
   )
  ]
  rendererClasses [&NiagaraSpriteRendererProperties, &NiagaraLightRendererProperties]
  renderers [
   {rendererClass=&NiagaraSpriteRendererProperties}
   {rendererClass=&NiagaraSpriteRendererProperties, rendererIndex=1}
   {rendererClass=&NiagaraLightRendererProperties, rendererIndex=2}
  ]
 }
 ImpactSmoke001 {rendererClasses [], renderers []}
 Shockwave {rendererClasses [], renderers []}
 Star_Core001 {rendererClasses [], renderers []}
}
summary {
 systemName=NS_Emote_WiryPerk_Impact
 emitters [
  {
   emitterName=3DBlocks_Horiz
   rendererClasses [
    &NiagaraSpriteRendererProperties
    &NiagaraMeshRendererProperties
    &NiagaraLightRendererProperties
   ]
  }
  {emitterName=3DBlocks_Vert, rendererClasses [&NiagaraSpriteRendererProperties, &NiagaraMeshRendererProperties]}
  {
   emitterName=GG_Collapse_GPU
   simTarget=GPUComputeSim
   rendererClasses [&NiagaraSpriteRendererProperties, &NiagaraMeshRendererProperties]
  }
  {
   emitterName=GG_Collapse_CPU
   rendererClasses [&NiagaraSpriteRendererProperties, &NiagaraMeshRendererProperties]
  }
  {emitterName=GG_Text, rendererClasses [&NiagaraSpriteRendererProperties, &NiagaraLightRendererProperties]}
  {emitterName=ImpactSmoke001, rendererClasses []}
  {emitterName=DirectionalFlare001, rendererClasses []}
  {emitterName=Star_Core001, rendererClasses []}
  {emitterName=Flash, rendererClasses []}
  {emitterName=Shockwave, rendererClasses []}
 ]
 userVariables [
  User.bIsFrontEnd {type=NiagaraBool}
  User.bIsFrontEndPreview {type=NiagaraBool}
  User.FrontEndScale {type=NiagaraFloat, defaultValue=0, description=""}
 ]
}
